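# Editor utility that bakes the Mannequin's AnimToTexture animation textures in both vertex and bone modes and updates the matching materials.
# Run sequences the vertex data asset pass, then the bone pass.
# Each pass sets the static mesh's lightmap index with generated lightmap UVs, then calls AnimationToTexture.
# On success, it pushes the data asset's values as global parameters into that mode's Body and ChestLog material instances.
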EditorUtilityBlueprint /AnimToTexture/Characters/Mannequin/BP_AnimToTexture
graphs [EventGraph {nodes [
   Event Run()
   > n1 = Sequence()
   n2 = Get BoneDataAsset() -> (BoneDataAsset:Anim To Texture Data Asset Object Reference)
   SetLightMapIndex!(
    execute=@n1.then_1
    StaticMesh:Static Mesh Object Reference=Get Static Mesh!(self:Anim To Texture Data Asset Object Reference=@n2) -> (ReturnValue:Static Mesh Object Reference) {is=title K2Node_CallFunction}
    LODIndex:int=Get StaticLODIndex!(self:Anim To Texture Data Asset Object Reference=@n2) -> (StaticLODIndex:int=0)
    LightmapIndex:int=2
    bGenerateLightmapUVs:bool=true
   ) -> (then, ReturnValue:bool)
   > n3 = AnimationToTexture!(execute, DataAsset:Anim To Texture Data Asset Object Reference=Get BoneDataAsset()) -> (then, ReturnValue:bool)
   > Branch(Condition=@n3)
   > UpdateMaterialInstanceFromDataAsset!(
    execute
    DataAsset:Anim To Texture Data Asset Object Reference=Get BoneDataAsset()
    MaterialInstance:Material Instance Constant Object Reference=/AnimToTexture/Characters/Mannequin/Materials/BoneAnimation/MI_Body_BoneAnimation
    MaterialParameterAssociation:EMaterialParameterAssociation Enum=GlobalParameter
   ) -> (then)
   > UpdateMaterialInstanceFromDataAsset(
    DataAsset=Get BoneDataAsset()
    MaterialInstance=/AnimToTexture/Characters/Mannequin/Materials/BoneAnimation/MI_ChestLog_BoneAnimation
    MaterialParameterAssociation=GlobalParameter
   )
   n4 = Get VertexDataAsset() -> (VertexDataAsset:Anim To Texture Data Asset Object Reference)
   SetLightMapIndex(
    execute=@n1.then_0
    StaticMesh=Get Static Mesh(self=@n4)
    LODIndex=Get StaticLODIndex(self=@n4) -> (StaticLODIndex=0)
    LightmapIndex=2
    bGenerateLightmapUVs=true
   )
   > n5 = AnimationToTexture(DataAsset=Get VertexDataAsset())
   > Branch(Condition=@n5)
   > UpdateMaterialInstanceFromDataAsset(
    DataAsset=Get VertexDataAsset()
    MaterialInstance=/AnimToTexture/Characters/Mannequin/Materials/VertexAnimation/MI_Body_VertexAnimation
    MaterialParameterAssociation=GlobalParameter
   )
   > UpdateMaterialInstanceFromDataAsset(
    DataAsset=Get VertexDataAsset()
    MaterialInstance=/AnimToTexture/Characters/Mannequin/Materials/VertexAnimation/MI_ChestLog_VertexAnimation
    MaterialParameterAssociation=GlobalParameter
   )
  ]}]
variables [
 BoneDataAsset {type=AnimToTextureDataAsset, default=/AnimToTexture/Characters/Mannequin/Data/DA_BoneAnimation}
 VertexDataAsset {type=AnimToTextureDataAsset, default=/AnimToTexture/Characters/Mannequin/Data/DA_VertexAnimation}
]
