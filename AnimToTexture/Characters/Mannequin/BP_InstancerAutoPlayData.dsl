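; Mannequin instancer Blueprint that plays AnimToTexture animations on instanced static meshes in auto-play mode, with varied playback per instance.
; The construction script enables Auto Play, runs the parent script, then calls BuildAutoPlayData.
; That function gives each instance a random animation from the data asset, plus random time offset and play rate.
; A single batch update applies this data and the transforms to the instances; event graph nodes are empty.
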(event EventBeginPlay)

(event EventTick
  (params (DeltaSeconds :type Float%20%28single-precision%29)))

(event Collision|EventActorBeginOverlap
  (params (OtherActor :type Actor%20Object%20Reference)))

(fn BuildAutoPlayData
  (params (InstancedStaticMesh :type Instanced%20Static%20Mesh%20Component%20Object%20Reference) (DataAsset :type Anim%20To%20Texture%20Data%20Asset%20Object%20Reference))
  (Utilities|Array|Clear (LocalVariables|Get_AutoPlayData))
  (for _ (range (Components|InstancedStaticMesh|GetInstanceCount InstancedStaticMesh))
    (bind _autoplaydata (AnimToTexture|Playback|GetAutoPlayDatafromDataAsset DataAsset (Math|Random|RandomInteger (Utilities|Array|Length (Variables|/Script/AnimToTexture.AnimToTextureDataAsset|GetAnimations DataAsset))) (Math|Random|RandomFloatinRange 0.0 1e+01) (Math|Random|RandomFloatinRange 0.75 1.5)))
    (Utilities|Array|Add (LocalVariables|Get_AutoPlayData) _autoplaydata))
  (return (LocalVariables|Get_AutoPlayData)))

(fn ConstructionScript
  (Variables|Default|SetAuto%20Play true)
  (|CallParentFunction)
  (bind _autoplaydata (CallFunction|BuildAutoPlayData (Variables|Default|GetInstancedStaticMesh) (Variables|Default|GetDataAsset)))
  (AnimToTexture|Playback|BatchUpdateInstancesAutoPlayData (Variables|Default|GetInstancedStaticMesh) _autoplaydata (Variables|Default|GetTransforms)))
