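; Child Blueprint of an AnimToTexture instancer that animates instanced meshes from baked animation textures, with per-instance variation.
; The construction script disables Auto Play, runs the parent, then fills parallel arrays with random starting frame data, animation indices, time offsets and play rates.
; Each tick recomputes every instance's frame from the data asset and game time on the CPU, then batch-updates the instanced static mesh with frame data and transforms.
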(event EventTick
  (params (DeltaSeconds :type Float%20%28single-precision%29))
  (for _index (range (Components|InstancedStaticMesh|GetInstanceCount (Variables|Default|GetInstancedStaticMesh)))
    (bind _framedata (AnimToTexture|Playback|GetFrameDatafromDataAsset (Variables|Default|GetDataAsset) (Utilities|Array|Get(acopy) (Variables|Default|GetAnimIndices) _index) (Utilities|Time|GetGameTimeinSeconds) (Utilities|Array|Get(acopy) (Variables|Default|GetTimeOffsets) _index) (Utilities|Array|Get(acopy) (Variables|Default|GetPlayRates) _index)))
    (Utilities|Array|SetArrayElem (Variables|Default|GetFrameData) _index _framedata))
  (AnimToTexture|Playback|BatchUpdateInstancesFrameData (Variables|Default|GetInstancedStaticMesh) (Variables|Default|GetFrameData) (Variables|Default|GetTransforms)))

(event Collision|EventActorBeginOverlap
  (params (OtherActor :type Actor%20Object%20Reference)))

(fn BuildFrameData
  (params (InstancedStaticMesh :type Instanced%20Static%20Mesh%20Component%20Object%20Reference) (DataAsset :type Anim%20To%20Texture%20Data%20Asset%20Object%20Reference))
  (Utilities|Array|Clear (Variables|Default|GetFrameData))
  (Utilities|Array|Clear (Variables|Default|GetAnimIndices))
  (Utilities|Array|Clear (Variables|Default|GetTimeOffsets))
  (Utilities|Array|Clear (Variables|Default|GetPlayRates))
  (for _ (range (Components|InstancedStaticMesh|GetInstanceCount InstancedStaticMesh))
    (Utilities|Array|Add (Variables|Default|GetFrameData) (Utilities|Struct|MakeAnimtoTextureFrameData (Math|Random|RandomFloat)))
    (Utilities|Array|Add (Variables|Default|GetAnimIndices) (Math|Random|RandomInteger (Utilities|Array|Length (Variables|/Script/AnimToTexture.AnimToTextureDataAsset|GetAnimations DataAsset))))
    (Utilities|Array|Add (Variables|Default|GetTimeOffsets) (Math|Random|RandomFloat))
    (Utilities|Array|Add (Variables|Default|GetPlayRates) (Math|Random|RandomFloatinRange 0.5 1.5))))

(fn ConstructionScript
  (Variables|Default|SetAuto%20Play false)
  (|CallParentFunction)
  (CallFunction|BuildFrameData (Variables|Default|GetInstancedStaticMesh) (Variables|Default|GetDataAsset))
  (AnimToTexture|Playback|BatchUpdateInstancesFrameData (Variables|Default|GetInstancedStaticMesh) (Variables|Default|GetFrameData) (Variables|Default|GetTransforms)))
