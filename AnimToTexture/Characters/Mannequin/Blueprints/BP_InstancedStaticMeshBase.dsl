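; Base actor Blueprint that lays out a grid of instanced static meshes for AnimToTexture playback.
; The construction script stores transforms from BuildTransforms, then passes their count and the Auto Play flag to SetupInstancedMeshComponent.
; BuildTransforms spaces points evenly around the local origin via CoordinateAtIndex, adding random X/Y jitter scaled by RandX and RandY.
; Stored transforms are not applied to instances here, and the event graph is empty.
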(event EventBeginPlay)

(event EventTick
  (params (DeltaSeconds :type Float%20%28single-precision%29)))

(event Collision|EventActorBeginOverlap
  (params (OtherActor :type Actor%20Object%20Reference)))

(fn BuildTransforms
  (Utilities|Array|Clear (LocalVariables|GetPrivateTransforms))
  (Utilities|FlowControl|Sequence
    (:then_0
      (for _index (range (+ (Variables|Grid|GetDivX) 1))
        (for _index_2 (range (+ (Variables|Grid|GetDivY) 1))
          (Utilities|Array|Add (LocalVariables|GetPrivateTransforms) (Math|Transform|MakeTransform (Math|Vector|MakeVector (+ (CallFunction|CoordinateatIndex _index (Variables|Grid|GetSizeX) (Variables|Grid|GetDivX)) (* (Math|Random|RandomFloatinRange (Math|Conversions|ToFloat(Integer) (* (/ (Variables|Grid|GetSizeX) (Variables|Grid|GetDivX)) -1)) (Math|Conversions|ToFloat(Integer) (/ (Variables|Grid|GetSizeX) (Variables|Grid|GetDivX)))) (Variables|Grid|GetRandX))) (+ (CallFunction|CoordinateatIndex _index_2 (Variables|Grid|GetSizeY) (Variables|Grid|GetDivY)) (* (Math|Random|RandomFloatinRange (Math|Conversions|ToFloat(Integer) (* (/ (Variables|Grid|GetSizeY) (Variables|Grid|GetDivY)) -1)) (Math|Conversions|ToFloat(Integer) (/ (Variables|Grid|GetSizeY) (Variables|Grid|GetDivY)))) (Variables|Grid|GetRandY)))))))))
    (:then_1
      (return (LocalVariables|GetPrivateTransforms)))))

(fn CoordinateAtIndex
  (params (Index :type Integer) (Size :type Integer) (Div :type Integer))
  (return (+ (* -1.0 Size) (* (Math|Conversions|ToFloat(Integer) Index) (/ (Math|Conversions|ToFloat(Integer) (* Size 2)) (Math|Conversions|ToFloat(Integer) Div))))))

(fn ConstructionScript
  (bind _outtransforms (CallFunction|BuildTransforms))
  (bind _output_get (Variables|Default|SetTransforms _outtransforms))
  (AnimToTexture|Playback|SetupInstancedMeshComponent (Variables|Default|GetInstancedStaticMesh) (Utilities|Array|Length _output_get) (Variables|Default|GetAuto%20Play)))
